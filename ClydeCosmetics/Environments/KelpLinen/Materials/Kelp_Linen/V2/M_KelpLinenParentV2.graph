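# Masked parent material for KelpLinen skinned characters, layering event lighting, gradient and dissolve effects over the standard 2023 character parent function.
# Fresnel-masked rim light, underlight, hand glow and collection-driven glow color build the effects, with beat adjustments alongside an audio-reactive function.
# Pre-skinned-position sphere masks shape phase-tracked gradients, fades and a temporally dithered dissolve.
# A Phase1 static switch selects desaturated, flattened-surface blending or emissive-led shading with shadow-replaced opacity.
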Material /ClydeCosmetics/Environments/KelpLinen/Materials/Kelp_Linen/V2/M_KelpLinenParentV2
graph {
 properties {
  allowTranslucentCustomDepthWrites=true
  bUseMaterialAttributes=true
  bUsedWithClothing=true
  bUsedWithMeshDeformer=true
  bUsedWithSkeletalMesh=true
  blendMode=BLEND_Masked
  pixelDepthOffsetMode=PDOM_Legacy
  shadingModel=MSM_FromMaterialExpression
  usedShadingModels=Default Lit | Subsurface | Clear Coat | Subsurface Profile | Cloth
 }
 nodes [
  n1 = CollectionParameter() {collection=/ClydeCosmetics/Environments/KelpLinen/Materials/MPC_KelpLinen, parameterName=KelpLinen_Glow_Color}
  n2 = LinearInterpolate(
   A=Multiply(A=@n1, B=ScalarParameter() {defaultValue=0.1, group=Kelp Gradient, parameterName=Gradient_Multi})
   B=@n1
   Alpha=NamedRerouteUsage()
  )
  n3 = Multiply(A=NamedRerouteUsage(), B=ScalarParameter() {parameterName=Phase_1_Basecolor_Str})
  n4 = Constant()
  Add(A=@n2, B=Multiply(A=AppendVector(A=@n3, B=@n4), B=NamedRerouteUsage()))
  n5 = ScalarParameter() {defaultValue=2, group=RimLight, parameterName=RimLight_Str_Hair}
  n6 = ScalarParameter() {defaultValue=1, group=RimLight, parameterName=Rim_Beat_Adjust}
  n7 = Multiply(A=LinearInterpolate(A=@n5, B=Add(A=@n5, B=@n6), Alpha=NamedRerouteUsage()), B=NamedRerouteUsage())
  n8 = TextureSample() {
   samplerType=SAMPLERTYPE_Masks
   texture=/BRCosmetics/Characters/Player/Male/Medium/Bodies/M_MED_KelpLinenCalcium/Textures/T_KelpLinenCalcium_FaceAcc_FX
  }
  n9 = OneMinus(None=@n8.R)
  n10 = ScalarParameter() {defaultValue=2, group=RimLight, parameterName=RimLight_Str}
  n11 = LinearInterpolate(A=@n10, B=Add(A=@n10, B=ScalarParameter() {defaultValue=1, group=RimLight, parameterName=Rim_Beat_Adjust}), Alpha=NamedRerouteUsage())
  n12 = GetMaterialAttributes(None=MaterialFunctionCall() {materialFunction=&MF_CharacterParent_2023_PreFX}.Result) {attributeGetTypes [#1, #2, #3]}
  n13 = Transform(None=@n12.Normal)
  n14 = MaterialFunctionCall!(
   BaseColor=Constant() {r=1}
   Gmap=StaticSwitchParameter(True=LinearInterpolate(A=@n7, B=@n11, Alpha=@n9), False=@n11) {parameterName=FaceAccessory?}
   Normals=@n13
   LightVector=VectorParameter() {group=RimLight, parameterName=Light_Direction, defaultValue=<rgba -0.016 0.3797 0.304 1>}.RGB
  ) {materialFunction=/ClydeCosmetics/Environments/KelpLinen/MF_KelpLinen_Rimlight}
  n15 = NamedRerouteUsage()
  n16 = MaterialFunctionCall(
   In=Fresnel(ExponentIn=ScalarParameter() {defaultValue=0.7, group=RimLight, parameterName=Fresnel Exp}, Normal=@n13) {baseReflectFraction=0}
   Contrast=Constant() {r=0.03}
  ) {materialFunction=&CheapContrast}
  n17 = Multiply(A=StaticSwitchParameter(True=@n14.Result, False=Multiply(A=@n14.Result, B=@n15)) {parameterName=FaceAccessory?}, B=@n16.Result)
  Add(B=@n17)
  Constant()
  Constant() {r=5}
  n18 = NamedRerouteUsage()
  NamedRerouteDeclaration(None=ComponentMask(None=Multiply(A=@n1, B=@n18)) {b=true, g=true, r=true}) {name=Glow color, nodeColor=<rgba 0 0.1797 1 1>}
  NamedRerouteDeclaration(None=@n12.MaterialAttributes) {name=Base Mat, nodeColor=<rgba 0.5 1 0 1>}
  n19 = Multiply(A=Multiply(A=@n16.Result) {constB=0.25}, B=NamedRerouteUsage())
  NamedRerouteDeclaration(None=Saturate(None=Add(A=@n17, B=@n19))) {name=Fresnel MAsk, nodeColor=<rgba 1 0 0.5156 1>}
  NamedRerouteDeclaration(None=@n13) {name=Normal- World Space, nodeColor=<rgba 1 0.9844 0 1>}
  n20 = VectorParameter() {group=Hand_Light, parameterName=Color B, sortPriority=20, defaultValue=<rgba 0.0592 0.9473 0.5151 1>}
  n21 = ScalarParameter() {defaultValue=4, group=Hand_Light, parameterName=light intensity B}
  n22 = VectorParameter() {group=Hand_Light, parameterName=CenterPoint2, sortPriority=10, defaultValue=<rgba -25.2108 28.7725 118.6585 1>}
  n23 = SphereMask(
   A=NamedRerouteUsage()
   B=@n22.RGB
   Radius=ScalarParameter() {defaultValue=32.7146, group=Hand_Light, parameterName=HandGlow_Radius}
   Hardness=ScalarParameter() {defaultValue=0.8, group=Hand_Light, parameterName=handglow_Contrast}
  )
  n24 = DotProduct(A=Normalize(VectorInput=@n22.RGB), B=NamedRerouteUsage())
  NamedRerouteDeclaration(None=Multiply(A=Multiply(A=@n20.RGB, B=@n21), B=Multiply(A=@n23, B=@n24))) {name=Hand_Glow, nodeColor=<rgba 0.5 0 1 1>}
  n25 = MaterialFunctionCall() {materialFunction=/ClydeCosmetics/Environments/KelpLinen/MF_KelpLinen_LocationFade}
  n26 = NamedRerouteUsage()
  n27 = VertexInterpolator(VS=SphereMask(
    A=MaterialFunctionCall(Float3=@n26) {materialFunction=&BreakOutFloat3Components}.B
    B=ScalarParameter() {defaultValue=60, group=Kelp Opacity, parameterName=Dissolve Z offset}
    Radius=ScalarParameter() {defaultValue=119.04, group=Kelp Opacity, parameterName=dissolve Radius}
    Hardness=ScalarParameter() {defaultValue=0.2, group=Kelp Opacity, parameterName=Dissolve Softness}
   ))
  NamedRerouteDeclaration(None=MaterialFunctionCall(Alpha Threshold=MaterialFunctionCall(Transition Phase=@n25.Output, Height Texture=@n27.PS) {materialFunction=&HeightLerp}.Alpha) {materialFunction=&DitherTemporalAA}.Result) {name=Opacity Mask, nodeColor=<rgba 1 0.7734 0 1>}
  n28 = MaterialFunctionCall(Float3=NamedRerouteUsage()) {materialFunction=&BreakOutFloat3Components}
  MaterialExpressionConstant_0 = Constant()
  MaterialExpressionScalarParameter_28 = ScalarParameter() {defaultValue=122, desc=125 was previous value, group=Kelp Gradient, parameterName=Gradient Z offset}
  MaterialExpressionScalarParameter_16 = ScalarParameter() {defaultValue=112.0167, group=Kelp Gradient, parameterName=Gradient Z offset2}
  n29 = Switch!(
   SwitchValue=CollectionParameter() {collection=&MPC_KelpLinen, parameterName=Phase_Tracker}
   Default
   dummy=@MaterialExpressionConstant_0
   Phase1=@MaterialExpressionScalarParameter_28
   Phase2=@MaterialExpressionScalarParameter_16
  ) {inputs [
    {
     inputName=dummy
     input {
      expression=&self:MaterialExpressionConstant_0
      inputName=None
      mask=0
      maskA=0
      maskB=0
      maskG=0
      maskR=0
      outputIndex=0
     }
    }
    {
     inputName=Phase1
     input {
      expression=&self:MaterialExpressionScalarParameter_28
      inputName=None
      mask=0
      maskA=0
      maskB=0
      maskG=0
      maskR=0
      outputIndex=0
     }
    }
    {
     inputName=Phase2
     input {
      expression=&self:MaterialExpressionScalarParameter_16
      inputName=None
      mask=0
      maskA=0
      maskB=0
      maskG=0
      maskR=0
      outputIndex=0
     }
    }
   ]}
  n30 = ScalarParameter() {defaultValue=30.3053, group=Kelp Gradient, parameterName=Gradient Radius}
  n31 = ScalarParameter() {defaultValue=0.1, group=Kelp Gradient, parameterName=Gradient Softness}
  NamedRerouteDeclaration(None=VertexInterpolator(VS=SphereMask(A=@n28.B, B=@n29, Radius=@n30, Hardness=@n31)).PS) {name=Mannequin Gradient, nodeColor=<rgba 1 0.3047 0 1>}
  NamedRerouteDeclaration(None=MaterialFunctionCall!(In) {materialFunction=/ClydeCosmetics/Environments/KelpLinen/MF_Kelp_AudioReactive}.Beat_Tracker) {name=Music Beat, nodeColor=<rgba 0.1484 1 0 1>}
  NamedRerouteDeclaration(None=@n12.Base Color) {name=base color, nodeColor=<rgba 0 1 0.0391 1>}
  NamedRerouteDeclaration(None=@n12.Emissive Color) {name=Base Emissive, nodeColor=<rgba 1 0 0.5156 1>}
  n32 = LocalPosition() {localOrigin=InstancePreSkinning}
  NamedRerouteDeclaration(None=SphereMask(
    A=VertexInterpolator(VS=@n32.XYZ).PS
    B=VectorParameter() {group=ColorFade, parameterName=Color_Fade_Position, defaultValue=<rgba 0 4 167 1>}.RGB
    Radius=ScalarParameter() {group=ColorFade, parameterName=Color_Fade_Radius}
    Hardness=ScalarParameter() {defaultValue=1, group=ColorFade, parameterName=Color_contrast}
   )) {name=Color Mask, nodeColor=<rgba 0.8047 1 0 1>}
  n33 = PreSkinnedPosition()
  NamedRerouteDeclaration(None=ShaderStageSwitch(PixelShader=VertexInterpolator(VS=@n33).PS, VertexShader=@n33)) {name=Pre Skinned Local Postition, nodeColor=<rgba 1 0.5156 0 1>}
  n34 = ScalarParameter() {defaultValue=1.5, group=Kelp Gradient, parameterName=Rim_Adjustment}
  n35 = LinearInterpolate(A=@n34, B=Add(A=@n34, B=ScalarParameter() {defaultValue=0.5, group=RimLight, parameterName=Rim_Beat_Adjust2}), Alpha=NamedRerouteUsage())
  NamedRerouteDeclaration(None=LinearInterpolate(A=@n2, B=Multiply(A=@n2, B=@n35) {constB=0.25}, Alpha=@n18)) {name=Phase_1 Emissive, nodeColor=<rgba 0 0.8359 1 1>}
  n36 = VertexInterpolator(VS=LocalPosition() {localOrigin=InstancePreSkinning}.XYZ)
  n37 = VectorParameter() {group=RimLight, parameterName=Light_Fade_Position, defaultValue=<rgba 0 4 167 1>}
  n38 = ScalarParameter() {defaultValue=21.5, group=RimLight, parameterName=Light_Fade_Radius}
  n39 = ScalarParameter() {defaultValue=1, group=RimLight, parameterName=Light_contrast}
  NamedRerouteDeclaration(None=OneMinus(None=SphereMask(A=@n36.PS, B=@n37.RGB, Radius=@n38, Hardness=@n39))) {name=Fade_Mask, nodeColor=<rgba 0.8281 1 0 1>}
  NamedRerouteUsage()
  NamedRerouteUsage()
  PixelNormalWS()
  n40 = MaterialFunctionCall(
   BaseColor=Constant() {r=1}
   Gmap=ScalarParameter() {defaultValue=1.1398, parameterName=RimLight_Underlight}
   Normals=NamedRerouteUsage()
   LightVector=VectorParameter() {parameterName=Light_Direction_5, defaultValue=<rgba 0 -1 -0.3 1>}.RGB
  ) {materialFunction=&MF_KelpLinen_Rimlight}
  n41 = ScalarParameter() {defaultValue=0.5, parameterName=Underlight_Str}
  n42 = Multiply(
   A=SphereMask(
    A=NamedRerouteUsage()
    B=VectorParameter() {group=Kelp Underlight, parameterName=Underlight_Skeletal_Mask, defaultValue=<rgba 0 0 60 1>}.RGB
    Radius=ScalarParameter() {defaultValue=120, group=Kelp Underlight, parameterName=Light_Gradient_Radius}
    Hardness=ScalarParameter() {defaultValue=0.2, group=Kelp Underlight, parameterName=Light_Softness}
   )
   B=DotProduct(A=Multiply(A=@n40.Result, B=@n41), B=NamedRerouteUsage())
  ) {constA=1}
  Reroute(None=StaticSwitchParameter(True=Multiply(A=@n42, B=@n9), False=@n42) {parameterName=FaceAccessory?})
  n43 = GetMaterialAttributes(None=NamedRerouteUsage()) {attributeGetTypes [
    #2
    #4
    #5
    #1
   ]}
  n44 = NamedRerouteUsage()
  n45 = Add(
   A=Desaturation(None=NamedRerouteUsage(), Fraction=ScalarParameter() {defaultValue=1, group=ColorFade, parameterName=Kelp Desaturate Frac})
   B=Multiply(A=ComponentMask(None=@n44) {b=true, g=true, r=true}, B=ScalarParameter() {defaultValue=0.5, parameterName=Kelp_Fog_BaseColor_Str})
  )
  n46 = Saturate(None=Subtract(A=NamedRerouteUsage(), B=NamedRerouteUsage()))
  n47 = NamedRerouteUsage()
  n48 = ComponentMask(None=NamedRerouteUsage()) {b=true, g=true, r=true}
  n49 = NamedRerouteUsage()
  n50 = SetMaterialAttributes!(
   MaterialAttributes=@n43.MaterialAttributes
   Base Color=LinearInterpolate(A=@n43.Base Color, B=Saturate(None=@n45), Alpha=@n46)
   Roughness=LinearInterpolate(A=@n43.Roughness, Alpha=@n46)
   Specular=LinearInterpolate(A=@n43.Specular, Alpha=@n46) {constB=0.2}
   Normal=LinearInterpolate(A=@n43.Normal, B=Constant3Vector() {constant=<rgba 0 0 1 0>}, Alpha=@n46)
   Emissive Color=Add(A=NamedRerouteUsage(), B=LinearInterpolate(A=@n47, B=@n48, Alpha=@n49))
   Opacity Mask=NamedRerouteUsage()
  ) {attributeSetTypes [
    #2
    #4
    #5
    #1
    #3
    #6
   ]}
  n51 = NamedRerouteUsage()
  n52 = NamedRerouteUsage()
  n53 = Multiply(A=@n51, B=ScalarParameter() {parameterName=BaseColor_Influence, primitiveDataIndex=11})
  n54 = Add(
   A=LinearInterpolate(A=Multiply(A=@n51, B=@n52), B=Add(A=@n53, B=@n52), Alpha=ScalarParameter() {defaultValue=0.5, parameterName=lerp_Colors_})
   B=NamedRerouteUsage()
  )
  n55 = NamedRerouteUsage()
  n56 = NamedRerouteUsage()
  n57 = ScalarParameter() {defaultValue=0.2, parameterName=Base Color Multi}
  n58 = NamedRerouteUsage()
  n59 = Constant()
  StaticSwitchParameter(
   True=@n50
   False=SetMaterialAttributes!(
    MaterialAttributes=NamedRerouteUsage()
    Emissive Color=Add(A=@n54, B=@n55)
    Base Color=Multiply(A=@n56, B=@n57)
    Opacity Mask=ShadowReplace(Default=@n58, Shadow=@n59)
   ) {attributeSetTypes [#3, #2, #6]}
  ) {parameterName=Phase1}
 ]
}
